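# Niagara effect for the OatmealSpread Golem emote that launches a short-lived, warm glowing ball with trails and rings.
# Ball fires one arcing particle under local gravity and drag, drawn as two sprites plus a light.
# Ribbon (sprites despite its name) and Sparkles spawn from Ball through particle reads, adding a fading trail and curl-noise sparks.
# Main_Ring bursts three staggered, spinning, expanding meshes with per-group scale and offset; WindBurst is disabled.
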NiagaraSystem /BRCosmetics/Animation/Game/MainPlayer/Emotes/OatmealSpread_Golem/FX/NS_OatmealSpread_Golem_GlowBall
dependencies {
 usedDataInterfaces [
  &NiagaraDataInterfaceCurve
  &NiagaraDataInterfaceVector2DCurve
  &NiagaraDataInterfaceColorCurve
  &NiagaraDataInterfaceVectorCurve
  &NiagaraDataInterfaceParticleRead
 ]
 usedDynamicInputs []
 usedModules [&SystemState]
}
emitters {
 Ball {
  emitterUpdateScript [
   &EmitterState(Life Cycle Mode=::NewEnumerator1, Loop Behavior=::NewEnumerator1, Loop Duration=1, UseLoopDelay=false)
   &SpawnBurst_Instantaneous(Use Spawn Probability=false, UseLoopCountLimit=false)
  ]
  particleSpawnScript [
   &InitializeParticle(
    Color=<rgba 5 3.7378 1.6399 3>
    Lifetime=0.2
    Ribbon Width=25
    Ribbon Width Mode=::NewEnumerator1
    Sprite Size Mode=::NewEnumerator3
    Uniform Sprite Size=15
    UsePositionOffset=false
    Write Material Random=true
   )
   &AddVelocity(Velocity=<0 500 325>)
  ]
  particleUpdateScript [
   &GravityForce(Coordinate Space=::Local, Gravity=<0 0 -500>)
   &Drag(Use Linear Drag=true, Use Rotational Drag=true)
   &SolveForcesAndVelocity(Clamp Velocity=false, Limit Acceleration=false)
   &ParticleState()
   &ScaleSpriteSize(
    Non-Uniform Curve Sprite Scale=DataInterface() {propertyValues {
      XCurve {keys [
        <at 0 0 RCIM_Cubic>
        {arriveTangent=2.8, interpMode=RCIM_Cubic, leaveTangent=2.8, time=0.1, value=2}
        {arriveTangent=-0.0627, interpMode=RCIM_Cubic, leaveTangent=-0.0627, tangentMode=RCTM_User, time=0.25, value=0.7}
        {arriveTangent=-0.2063, interpMode=RCIM_Cubic, leaveTangent=-0.2063, tangentMode=RCTM_User, time=0.75, value=0.7}
        {arriveTangent=-15.3364, tangentMode=RCTM_User, time=1, value=0}
       ]}
      YCurve {keys [
        {arriveTangent=0.2932, interpMode=RCIM_Cubic, leaveTangent=0.2932, tangentMode=RCTM_User, time=0, value=0}
        {arriveTangent=8.1364, interpMode=RCIM_Cubic, leaveTangent=8.1364, tangentMode=RCTM_User, time=0.1, value=2}
        {arriveTangent=0.1196, interpMode=RCIM_Cubic, leaveTangent=0.1196, tangentMode=RCTM_User, time=0.25, value=1}
        {arriveTangent=-0.1258, interpMode=RCIM_Cubic, leaveTangent=-0.1258, tangentMode=RCTM_User, time=0.45, value=1.65}
        <at 0.8 1.65>
        <at 1 1.65>
       ]}
     }}
    Scale Sprite Size Mode=::NewEnumerator3
   )
   &ScaleRibbonWidth(Ribbon Width Scale=&FloatFromCurve(FloatCurve=DataInterface() {propertyValues {Curve {keys [
         {arriveTangent=17.7467, interpMode=RCIM_Cubic, leaveTangent=17.7467, tangentMode=RCTM_User, time=0, value=0}
         <at 0.1 1 RCIM_Cubic>
         {arriveTangent=0.1047, interpMode=RCIM_Cubic, leaveTangent=0.1047, tangentMode=RCTM_User, time=1, value=0}
        ]}}}))
   /SetVariables(
    Particles.BGColor:LinearColor=&ColorFromCurve(ColorCurve=DataInterface() {propertyValues {
       AlphaCurve {keys [
         {arriveTangent=22.7767, interpMode=RCIM_Cubic, leaveTangent=22.7767, tangentMode=RCTM_User, time=0, value=0}
         {arriveTangentWeight=0.0333, interpMode=RCIM_Cubic, tangentMode=RCTM_User, time=0.1, value=0.1}
         {arriveTangent=-0.6219, arriveTangentWeight=0.3, interpMode=RCIM_Cubic, leaveTangent=-0.6219, tangentMode=RCTM_User, time=1, value=0}
        ]}
       BlueCurve {keys [<at 0 0.2604 RCIM_Cubic>]}
       GreenCurve {keys [<at 0 0.8621 RCIM_Cubic>]}
       RedCurve {keys [<at 0 1 RCIM_Cubic>]}
      }})
    Particles.BGSpriteSize:Vector2f=&Vector2DFromFloat(Value=&FloatFromCurve(
      FloatCurve=DataInterface() {propertyValues {Curve {keys [
          <at 0 0 RCIM_Cubic>
          {arriveTangent=10, interpMode=RCIM_Cubic, leaveTangent=10, time=0.05, value=1.2}
          {arriveTangent=-0.2667, interpMode=RCIM_Cubic, leaveTangent=-0.2667, time=0.1, value=1}
          {arriveTangent=-0.2819, interpMode=RCIM_Cubic, leaveTangent=-0.2819, tangentMode=RCTM_User, time=0.8, value=1}
          {arriveTangent=-0.0931, interpMode=RCIM_Cubic, leaveTangent=-0.0931, tangentMode=RCTM_User, time=1, value=0}
         ]}}}
      Scale Curve=80
     ))
   ) {isSetParameters=true}
   &ScaleColor(Scale Mode=::NewEnumerator2)
   &FN_Light_Attributes(
    Color=<rgba 1 0.7476 0.328 1>
    Create Color Attribute=true
    Create Exponent Attribute=true
    Create Radius Attribute=true
    Exponent=4
    Radius=250
    Scale Color=&FloatFromCurve(
     FloatCurve=DataInterface() {propertyValues {Curve {keys [<at 0 1 RCIM_Cubic>, {arriveTangent=-2.6045, interpMode=RCIM_Cubic, leaveTangent=-2.6045, tangentMode=RCTM_User, time=1, value=0}]}}}
     Scale Curve=3
    )
   )
  ]
  rendererClasses [&NiagaraSpriteRendererProperties, &NiagaraLightRendererProperties]
  renderers [
   {rendererClass=&NiagaraSpriteRendererProperties}
   {rendererClass=&NiagaraSpriteRendererProperties, rendererIndex=1}
   {rendererClass=&NiagaraLightRendererProperties, rendererIndex=2}
  ]
 }
 Main_Ring {
  emitterUpdateScript [
   &EmitterState(Life Cycle Mode=::NewEnumerator1, Loop Behavior=::NewEnumerator1, Loop Delay=0.01, Loop Duration=1, UseLoopDelay=false)
   &SpawnBurst_Instantaneous(Use Spawn Probability=false, UseLoopCountLimit=false)
   &SpawnBurst_Instantaneous(Spawn Group=1, Spawn Time=0.08, Use Spawn Probability=false, UseLoopCountLimit=false) {module=SpawnBurst_Instantaneous001}
   &SpawnBurst_Instantaneous(Spawn Group=2, Spawn Time=0.16, Use Spawn Probability=false, UseLoopCountLimit=false) {module=SpawnBurst_Instantaneous002}
  ]
  particleSpawnScript [
   /SetVariables(Particles.RandomRange:NiagaraFloat=&Multiply_Float(
     A=&RandomRangeFloat(Minimum=0.25, Maximum=0.375)
     B=&FloatFromCurve(
      FloatCurve=DataInterface() {propertyValues {Curve {keys [<at 0 1 RCIM_Cubic>, {arriveTangent=7.3942, interpMode=RCIM_Cubic, leaveTangent=7.3942, tangentMode=RCTM_User, time=0.3, value=2}]}}}
      CurveIndex=$Emitter.Age
     )
    )) {isSetParameters=true}
   &InitializeParticle(
    Color=&FNTimeOfDayColor(Night Color=<rgba 1.5 1.5 1.5 1>, Morning Color=<rgba 3 3 3 1>, Day Color=<rgba 3 3 3 1>, Evening Color=<rgba 1.5 1.5 1.5 1>)
    Lifetime=0.45
    Mesh Scale=&Multiply_VectorByFloat(
     Vector=<1 1.25 1>
     Float=&MaskFloatBySpawnGroup(
      Masked Float=0.25
      Passthrough Float=&MaskFloatBySpawnGroup(Masked Float=0.15, Passthrough Float=0.1, Spawn Group Mask=1)
     )
    )
    Mesh Scale Mode=::NewEnumerator4
    Position Offset=&MaskVectorBySpawnGroup(
     Masked Vector=<0 3.5 5>
     Passthrough Vector=&MaskVectorBySpawnGroup(Masked Vector=<0 20 12>, Passthrough Vector=<1 35 22>, Spawn Group Mask=1)
    )
    UsePositionOffset=true
    Write Material Random=true
   )
   &InitialMeshOrientation(
    Mesh Orientation Mode=::NewEnumerator1
    Rotation=&MaskVectorBySpawnGroup(Masked Vector=<0.05 0 0>, Passthrough Vector=<0.06 0 0>)
    Use Rotation Vector=true
   )
  ]
  particleUpdateScript [
   &ParticleState()
   &ScaleMeshSize(Scale Factor=&VectorFromFloat(Value=&FloatFromCurve(
      FloatCurve=DataInterface() {propertyValues {Curve {keys [
          {arriveTangent=5.6347, interpMode=RCIM_Cubic, leaveTangent=5.6347, tangentMode=RCTM_User, time=0, value=0.3788}
          {arriveTangent=1.1451, interpMode=RCIM_Cubic, leaveTangent=1.1451, time=0.2, value=0.85}
          {arriveTangent=0.2133, interpMode=RCIM_Cubic, leaveTangent=0.2133, tangentMode=RCTM_User, time=0.6186, value=1.0872}
          <at 0.9985 1.2296 RCIM_Cubic>
         ]}}}
      Scale Curve=0.95
     )))
   &DynamicMaterialParameters(
    Index 0 Param 1=&FloatFromCurve(FloatCurve=DataInterface() {propertyValues {Curve {keys [
         {arriveTangent=2.3202, interpMode=RCIM_Cubic, leaveTangent=2.3202, tangentMode=RCTM_User, time=0, value=0.2481}
         {arriveTangent=0.6377, interpMode=RCIM_Cubic, leaveTangent=0.6377, time=0.0995, value=0.4178}
         {interpMode=RCIM_Cubic, tangentMode=RCTM_User, time=0.3951, value=0.5}
         {arriveTangent=-1.1781, tangentMode=RCTM_User, time=1, value=0}
        ]}}})
    Module.Index0Param0Write=true
    Module.Index0Param1Write=false
    Module.Index0Param2Write=false
    Module.Index0Param3Write=false
   )
   &ScaleColor(
    Scale Alpha=&FloatFromCurve(
     FloatCurve=DataInterface() {propertyValues {Curve {keys [
         {arriveTangent=-0.0871, interpMode=RCIM_Cubic, leaveTangent=-0.0871, tangentMode=RCTM_User, time=0, value=0.6231}
         {interpMode=RCIM_Cubic, tangentMode=RCTM_User, time=0.5, value=0.7329}
         {tangentMode=RCTM_User, time=1, value=0}
        ]}}}
     Scale Curve=&FloatFromCurve(
      FloatCurve=DataInterface() {propertyValues {Curve {keys [
          {interpMode=RCIM_Cubic, tangentMode=RCTM_User, time=0, value=1}
          {arriveTangent=-0.0131, interpMode=RCIM_Cubic, leaveTangent=0.1095, tangentMode=RCTM_Break, time=0.3, value=1}
          {arriveTangent=-3.0223, interpMode=RCIM_Cubic, leaveTangent=-3.0223, tangentMode=RCTM_User, time=0.7, value=0}
         ]}}}
      CurveIndex=$Emitter.Age
      Scale Curve=0.4
     )
    )
    Scale RGB=&MakeVectorFromLinearColorRGB(Color=&ColorFromCurve(ColorCurve=DataInterface() {propertyValues {
        AlphaCurve {keys []}
        BlueCurve {keys [
          <at 0 0.2188 RCIM_Cubic>
          {arriveTangent=-0.1912, interpMode=RCIM_Cubic, leaveTangent=-0.1912, time=0.1313, value=0.2738}
          <at 0.4231 0.1378>
          <at 1 0.0182>
         ]}
        GreenCurve {keys [
          <at 0 0.8989 RCIM_Cubic>
          {arriveTangent=-1.3071, interpMode=RCIM_Cubic, leaveTangent=-1.3071, time=0.1313, value=0.8997}
          <at 0.4231 0.3459>
          <at 1 0.1094>
         ]}
        RedCurve {keys [
          <at 0 1 RCIM_Cubic>
          {arriveTangent=-1.2161, interpMode=RCIM_Cubic, leaveTangent=-1.2161, time=0.1313, value=1}
          <at 0.4231 0.4855>
          <at 1 0.2309>
         ]}
       }}))
    ScaleA=true
    ScaleRGB=true
   )
   &UpdateMeshOrientation(
    Rotation Rate=&MaskFloatBySpawnGroup(
     Masked Float=1.25
     Passthrough Float=&MaskFloatBySpawnGroup(Masked Float=0.75, Passthrough Float=0.25, Spawn Group Mask=1)
    )
    Rotation Vector=<0 1 0>
   ) {module=UpdateMeshOrientation001}
  ]
  rendererClasses [&NiagaraMeshRendererProperties]
  renderers [{rendererClass=&NiagaraMeshRendererProperties, rendererIndex=0}]
 }
 Ribbon {
  emitterUpdateScript [
   &EmitterState(
    Enable Distance Culling=true
    Life Cycle Mode=::NewEnumerator1
    Loop Behavior=::NewEnumerator1
    Loop Duration=2.25
    MaxDistance=2500000
    Module.bUseMaxDistance=true
    Module.bUseMinDistance=false
    Scalability Mode=::NewEnumerator1
    UseLoopDelay=false
   )
   &SpawnParticlesFromOtherEmitter(
    Attribute Reader=DataInterface() {propertyValues {EmitterBinding {bindingMode=Other, emitterName=Ball}}}
    Spawn Rate=150
    Spawn Rate Per Particle Cap=150
    Use Maximum Age=false
    Use Minimum Age=false
   )
  ]
  particleSpawnScript [
   &InitializeParticle(
    Color=<rgba 1 1 1 0.15>
    Lifetime=0.475
    Sprite Size Mode=::NewEnumerator3
    Uniform Sprite Size=7
    UsePositionOffset=false
    Write Material Random=true
   )
   &SampleParticlesFromOtherEmitter(
    Age Sampling:ENiagara_DirectReadSamplingMode=::NewEnumerator0
    Lifetime Application:ENiagara_LifetimeInheritanceOptions=::NewEnumerator1
    Lifetime Sampling:ENiagara_DirectReadSamplingMode=::NewEnumerator0
    Particle ID Sampling=::NewEnumerator1
   )
  ]
  particleUpdateScript [
   &ParticleState()
   &ScaleSpriteSize(Uniform Curve Sprite Scale=DataInterface() {propertyValues {Curve {keys [<at 0 1 RCIM_Cubic>, {arriveTangent=-3.1899, interpMode=RCIM_Cubic, leaveTangent=-3.1899, tangentMode=RCTM_User, time=1, value=0}]}}})
   &ScaleColor(
    Linear Color Curve=DataInterface() {propertyValues {
      AlphaCurve {keys [
        <at 0 0 RCIM_Cubic>
        {arriveTangent=-0.1027, arriveTangentWeight=0.0333, interpMode=RCIM_Cubic, leaveTangent=-0.1027, tangentMode=RCTM_User, time=0.1, value=1}
        {arriveTangent=-1.1111, interpMode=RCIM_Cubic, leaveTangent=-1.1111, time=0.3762, value=0.3007}
        <at 1 0 RCIM_Cubic>
       ]}
      BlueCurve {keys [<at 0.2 0.1707>,<at 0.9 0.1452>]}
      GreenCurve {keys [<at 0.2 0.949>,<at 0.9 0.2314>]}
      RedCurve {keys [<at 0.2 1>,<at 0.9 1>]}
     }}
    Scale Mode=::NewEnumerator2
   ) {module=ScaleColor001}
   &SolveForcesAndVelocity(Clamp Velocity=false, Limit Acceleration=false)
  ]
  rendererClasses [&NiagaraSpriteRendererProperties]
  renderers [{rendererClass=&NiagaraSpriteRendererProperties}]
 }
 Sparkles {
  emitterUpdateScript [
   &EmitterState(
    Enable Distance Culling=true
    Life Cycle Mode=::NewEnumerator1
    Loop Behavior=::NewEnumerator1
    Loop Duration=2.25
    MaxDistance=2500000
    Module.bUseMaxDistance=true
    Module.bUseMinDistance=false
    Scalability Mode=::NewEnumerator1
    UseLoopDelay=false
   )
   &SpawnParticlesFromOtherEmitter(
    Attribute Reader=DataInterface() {propertyValues {EmitterBinding {bindingMode=Other, emitterName=Ball}}}
    Spawn Rate=65
    Spawn Rate Per Particle Cap=150
    Use Maximum Age=false
    Use Minimum Age=false
   )
  ]
  particleSpawnScript [
   &InitializeParticle(
    Color Maximum=<rgba 1 0.8512 0.1344 1>
    Color Minimum=<rgba 1 0.3359 0.1724 1>
    Color Mode=::NewEnumerator2
    Lifetime Max=0.7
    Lifetime Min=0.2
    Lifetime Mode=::NewEnumerator1
    Ribbon Facing Mode=::NewEnumerator1
    Sprite Rotation Mode=::NewEnumerator2
    Sprite Size Mode=::NewEnumerator1
    Uniform Sprite Size Max=25
    Uniform Sprite Size Min=10
    UsePositionOffset=false
    Write Material Random=true
   )
   &SampleParticlesFromOtherEmitter(
    Age Sampling:ENiagara_DirectReadSamplingMode=::NewEnumerator0
    Applied Velocity Scale=-0.2
    Lifetime Application:ENiagara_LifetimeInheritanceOptions=::NewEnumerator1
    Lifetime Sampling:ENiagara_DirectReadSamplingMode=::NewEnumerator0
    Particle ID Sampling=::NewEnumerator1
    Velocity Sampling=::NewEnumerator0
   )
   &ShapeLocation(MaskSpawnGroup=false, Override Seed=false, Sphere Radius=10, UseOwnerScale=true)
  ]
  particleUpdateScript [
   &ParticleState()
   &ScaleSpriteSize(Uniform Curve Sprite Scale=DataInterface() {propertyValues {Curve {keys [
        {arriveTangent=3.5564, interpMode=RCIM_Cubic, leaveTangent=3.5564, tangentMode=RCTM_User, time=0, value=0}
        {arriveTangent=0.0572, interpMode=RCIM_Cubic, leaveTangent=0.0572, tangentMode=RCTM_User, time=0.6, value=1}
        {arriveTangent=0.0145, interpMode=RCIM_Cubic, leaveTangent=0.0145, tangentMode=RCTM_User, time=0.8, value=1}
        {arriveTangent=-10.3135, interpMode=RCIM_Cubic, leaveTangent=-10.3135, tangentMode=RCTM_User, time=1, value=0}
       ]}}})
   &SpriteRotationRate(Rotation Rate=&RandomRangeFloat(Minimum=-3, Maximum=3))
   &CurlNoiseForce(Bias Noise=false, Noise Strength=250, Pan Noise=false)
   &GravityForce(Gravity=&RandomRangeVector(Minimum=<0 0 -90>, Maximum=<0 0 -200>))
   &Drag(
    Drag=&RandomRangeFloat(Minimum=1, Maximum=3)
    Rotational Drag=0
    Use Linear Drag=true
    Use Rotational Drag=true
   )
   &SolveForcesAndVelocity(Clamp Velocity=false, Limit Acceleration=false)
   &ScaleColor(
    Scale Alpha=&FloatFromCurve(FloatCurve=DataInterface() {propertyValues {Curve {keys [
         <at 0 1 RCIM_Cubic>
         {arriveTangent=-0.1027, interpMode=RCIM_Cubic, leaveTangent=-0.1027, tangentMode=RCTM_User, time=0.1, value=1}
         {arriveTangent=-2.6316, interpMode=RCIM_Cubic, leaveTangent=-2.6316, tangentMode=RCTM_User, time=1, value=0}
        ]}}})
    ScaleA=true
    ScaleRGB=true
   ) {module=ScaleColor001}
  ]
  rendererClasses [&NiagaraSpriteRendererProperties]
  renderers [{rendererClass=&NiagaraSpriteRendererProperties}]
 }
 WindBurst {
  bEnabled=false
  emitterUpdateScript [
   &EmitterState()
   &SpawnBurst_Instantaneous(Use Spawn Probability=false, UseLoopCountLimit=false)
  ]
  particleSpawnScript [
   &InitializeParticle(
    Color=<rgba 2 1.6446 0.7406 1>
    Lifetime=0.2
    Mesh Scale=<0.175 0.2 0.2>
    Mesh Scale Mode=::NewEnumerator4
    UsePositionOffset=true
    Write Material Random=true
   )
   &InitialMeshOrientation(Mesh Orientation Mode=::NewEnumerator1, Rotation=<-0.16 0 0>, Use Rotation Vector=true)
  ]
  particleUpdateScript [
   &ParticleState()
   &ScaleColor(
    Scale Alpha=&FloatFromCurve(FloatCurve=DataInterface() {propertyValues {Curve {keys [
         {arriveTangent=0.002, interpMode=RCIM_Cubic, leaveTangent=0.002, time=0, value=0}
         {arriveTangent=0.002, interpMode=RCIM_Cubic, leaveTangent=0.002, time=0.1, value=1}
         {arriveTangent=0.002, interpMode=RCIM_Cubic, leaveTangent=0.002, time=1.0011, value=0.002}
        ]}}})
    ScaleA=true
    ScaleRGB=true
   )
   &ScaleMeshSize(Scale Factor=&VectorFromCurve(VectorCurve=DataInterface() {propertyValues {
       XCurve {keys [{arriveTangent=2.4663, interpMode=RCIM_Cubic, leaveTangent=2.4663, tangentMode=RCTM_User, time=0, value=0}, <at 1 1 RCIM_Cubic>]}
       YCurve {keys [<at 0 0>,<at 1 1>]}
       ZCurve {keys [<at 0 0>,<at 1 1>]}
      }}))
  ]
  rendererClasses [&NiagaraMeshRendererProperties]
  renderers [{rendererClass=&NiagaraMeshRendererProperties, rendererIndex=0}]
 }
}
summary {
 systemName=NS_OatmealSpread_Golem_GlowBall
 emitters [
  {emitterName=Ball, rendererClasses [&NiagaraSpriteRendererProperties, &NiagaraLightRendererProperties]}
  {bEnabled=false, emitterName=WindBurst, rendererClasses [&NiagaraMeshRendererProperties]}
  {emitterName=Sparkles, rendererClasses [&NiagaraSpriteRendererProperties]}
  {emitterName=Ribbon, rendererClasses [&NiagaraSpriteRendererProperties]}
  {emitterName=Main_Ring, rendererClasses [&NiagaraMeshRendererProperties]}
 ]
 userVariables []
}
