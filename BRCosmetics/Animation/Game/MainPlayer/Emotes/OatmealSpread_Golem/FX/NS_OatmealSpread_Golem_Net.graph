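# Niagara effect for the OatmealSpread Golem emote's "Net" moment: a short, warm gold burst with a falling ball and trail.
# Ball launches a glowing sprite downward, and Ribbon reads its particles to spawn a fading sprite trail.
# Slightly delayed, CoreFlash pops a mesh, Sparkles scatter under curl noise, and Ring expands spinning meshes and sprites.
# Ring splits three bursts by spawn group via VisibilityTag, with separate glow, speedline and wind attributes.
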NiagaraSystem /BRCosmetics/Animation/Game/MainPlayer/Emotes/OatmealSpread_Golem/FX/NS_OatmealSpread_Golem_Net
dependencies {
 usedDataInterfaces [
  &NiagaraDataInterfaceCurve
  &NiagaraDataInterfaceColorCurve
  &NiagaraDataInterfaceVector2DCurve
  &NiagaraDataInterfaceParticleRead
 ]
 usedDynamicInputs []
 usedModules [&SystemState]
}
emitters {
 Ball {
  emitterUpdateScript [
   &EmitterState()
   &SpawnBurst_Instantaneous(Use Spawn Probability=false, UseLoopCountLimit=false)
  ]
  particleSpawnScript [
   &InitializeParticle(
    Color=<rgba 5 3.7378 1.6399 5>
    Lifetime=0.25
    Position Offset=<0 -15 50>
    Ribbon Width=25
    Ribbon Width Mode=::NewEnumerator1
    Sprite Size Mode=::NewEnumerator3
    Uniform Sprite Size=16
    UsePositionOffset=true
    Write Material Random=true
   )
   &AddVelocity(Velocity=<0 125 -325>)
  ]
  particleUpdateScript [
   &GravityForce(Coordinate Space=::Local)
   &Drag(Use Linear Drag=true, Use Rotational Drag=true)
   &SolveForcesAndVelocity(Clamp Velocity=false, Limit Acceleration=false)
   &ParticleState()
   &ScaleSpriteSize(
    Non-Uniform Curve Sprite Scale=DataInterface() {propertyValues {
      XCurve {keys [
        <at 0 0.5 RCIM_Cubic>
        {arriveTangent=-0.2063, interpMode=RCIM_Cubic, leaveTangent=-0.2063, tangentMode=RCTM_User, time=0.75, value=0.7}
        {arriveTangent=-5.3803, tangentMode=RCTM_User, time=1, value=0}
       ]}
      YCurve {keys [
        {arriveTangent=0.2932, interpMode=RCIM_Cubic, leaveTangent=0.2932, tangentMode=RCTM_User, time=0, value=0.5}
        {arriveTangent=-0.1258, interpMode=RCIM_Cubic, leaveTangent=-0.1258, tangentMode=RCTM_User, time=0.45, value=1.65}
        {arriveTangent=-0.0316, interpMode=RCIM_Cubic, leaveTangent=-0.0316, tangentMode=RCTM_User, time=0.8, value=1.65}
        {arriveTangent=-18.9212, interpMode=RCIM_Cubic, leaveTangent=-18.9212, tangentMode=RCTM_User, time=1, value=0}
       ]}
     }}
    Scale Sprite Size Mode=::NewEnumerator3
   )
   /SetVariables(
    Particles.BGColor:LinearColor=&ColorFromCurve(ColorCurve=DataInterface() {propertyValues {
       AlphaCurve {keys [
         {arriveTangent=22.7767, interpMode=RCIM_Cubic, leaveTangent=22.7767, tangentMode=RCTM_User, time=0, value=0}
         {arriveTangentWeight=0.0333, interpMode=RCIM_Cubic, tangentMode=RCTM_User, time=0.1, value=0.1}
         {
          arriveTangent=-0.1586
          arriveTangentWeight=0.2333
          interpMode=RCIM_Cubic
          leaveTangent=-0.1586
          leaveTangentWeight=0.0667
          tangentMode=RCTM_User
          time=0.8
          value=0.1
         }
         {arriveTangent=-0.6219, arriveTangentWeight=0.3, interpMode=RCIM_Cubic, leaveTangent=-0.6219, tangentMode=RCTM_User, time=1, value=0}
        ]}
       BlueCurve {keys [<at 0 0.2604 RCIM_Cubic>]}
       GreenCurve {keys [<at 0 0.8621 RCIM_Cubic>]}
       RedCurve {keys [<at 0 1 RCIM_Cubic>]}
      }})
    Particles.BGSpriteSize:Vector2f=&Vector2DFromFloat(Value=&FloatFromCurve(
      FloatCurve=DataInterface() {propertyValues {Curve {keys [
          <at 0 0 RCIM_Cubic>
          {arriveTangent=10, interpMode=RCIM_Cubic, leaveTangent=10, time=0.05, value=1.2}
          {arriveTangent=-0.2667, interpMode=RCIM_Cubic, leaveTangent=-0.2667, time=0.1, value=1}
          {arriveTangent=-0.2819, interpMode=RCIM_Cubic, leaveTangent=-0.2819, tangentMode=RCTM_User, time=0.8, value=1}
          {arriveTangent=-0.0931, interpMode=RCIM_Cubic, leaveTangent=-0.0931, tangentMode=RCTM_User, time=1, value=0}
         ]}}}
      Scale Curve=80
     ))
   ) {isSetParameters=true}
   &ScaleColor(Scale Mode=::NewEnumerator2)
  ]
  rendererClasses [&NiagaraSpriteRendererProperties]
  renderers [{rendererClass=&NiagaraSpriteRendererProperties}]
 }
 CoreFlash {
  emitterUpdateScript [
   &EmitterState()
   &SpawnBurst_Instantaneous(Spawn Count=1, Spawn Time=0.08, Use Spawn Probability=false)
  ]
  particleSpawnScript [
   &InitializeParticle(
    Color=<rgba 4 3.1843 0.7373 1>
    Lifetime=0.16
    Mesh Scale=<0.5 0.7 0.7>
    Mesh Scale Mode=::NewEnumerator4
    Position Offset=<0 -22 -5>
    UsePositionOffset=false
    Write Material Random=true
   )
   &InitialMeshOrientation(Mesh Orientation Mode=::NewEnumerator1, Rotation=<0 0.25 0>, Rotation Coordinate Space=::NewEnumerator2, Use Rotation Vector=true)
  ]
  particleUpdateScript [
   &ParticleState()
   &ScaleColor(
    Scale RGB=&VectorFromFloat(Value=&FloatFromCurve(FloatCurve=DataInterface() {propertyValues {Curve {keys [
          {interpMode=RCIM_Cubic, leaveTangent=-2.0665, tangentMode=RCTM_Break, time=0, value=1}
          {interpMode=RCIM_Cubic, tangentMode=RCTM_Break, time=1, value=0.01}
         ]}}}))
    ScaleA=true
    ScaleRGB=true
   )
   &ScaleMeshSize(Scale Factor=&VectorFromFloat(Value=&FloatFromCurve(FloatCurve=DataInterface() {propertyValues {Curve {keys [{arriveTangent=2.2457, interpMode=RCIM_Cubic, leaveTangent=2.2457, tangentMode=RCTM_User, time=0, value=0}, <at 1 1>]}}})))
   &SolveForcesAndVelocity(Clamp Velocity=false, Limit Acceleration=false)
  ]
  rendererClasses [&NiagaraMeshRendererProperties]
  renderers [{rendererClass=&NiagaraMeshRendererProperties, rendererIndex=0}]
 }
 Ribbon {
  emitterUpdateScript [
   &EmitterState(
    Enable Distance Culling=true
    Life Cycle Mode=::NewEnumerator1
    Loop Behavior=::NewEnumerator1
    Loop Duration=2.25
    MaxDistance=2500000
    Module.bUseMaxDistance=true
    Module.bUseMinDistance=false
    Scalability Mode=::NewEnumerator1
    UseLoopDelay=false
   )
   &SpawnParticlesFromOtherEmitter(
    Attribute Reader=DataInterface() {propertyValues {EmitterBinding {bindingMode=Other, emitterName=Ball}}}
    Spawn Rate=150
    Spawn Rate Per Particle Cap=150
    Use Maximum Age=false
    Use Minimum Age=false
   )
  ]
  particleSpawnScript [
   &InitializeParticle(
    Color=<rgba 1 1 1 0.1>
    Lifetime=0.475
    Sprite Size Mode=::NewEnumerator3
    Uniform Sprite Size=7
    UsePositionOffset=false
    Write Material Random=true
   )
   &SampleParticlesFromOtherEmitter(
    Age Sampling:ENiagara_DirectReadSamplingMode=::NewEnumerator0
    Lifetime Application:ENiagara_LifetimeInheritanceOptions=::NewEnumerator1
    Lifetime Sampling:ENiagara_DirectReadSamplingMode=::NewEnumerator0
    Particle ID Sampling=::NewEnumerator1
   )
  ]
  particleUpdateScript [
   &ParticleState()
   &ScaleSpriteSize(Uniform Curve Sprite Scale=DataInterface() {propertyValues {Curve {keys [<at 0 1 RCIM_Cubic>, {arriveTangent=-3.1899, interpMode=RCIM_Cubic, leaveTangent=-3.1899, tangentMode=RCTM_User, time=1, value=0}]}}})
   &ScaleColor(
    Linear Color Curve=DataInterface() {propertyValues {
      AlphaCurve {keys [
        <at 0 0 RCIM_Cubic>
        {arriveTangent=-0.1027, arriveTangentWeight=0.0333, interpMode=RCIM_Cubic, leaveTangent=-0.1027, tangentMode=RCTM_User, time=0.1, value=1}
        {arriveTangent=-1.1111, interpMode=RCIM_Cubic, leaveTangent=-1.1111, time=0.3762, value=0.3007}
        <at 1 0 RCIM_Cubic>
       ]}
      BlueCurve {keys [<at 0.2 0.1707>,<at 0.9 0.1452>]}
      GreenCurve {keys [<at 0.2 0.949>,<at 0.9 0.2314>]}
      RedCurve {keys [<at 0.2 1>,<at 0.9 1>]}
     }}
    Scale Mode=::NewEnumerator2
   ) {module=ScaleColor001}
   &SolveForcesAndVelocity(Clamp Velocity=false, Limit Acceleration=false)
  ]
  rendererClasses [&NiagaraSpriteRendererProperties]
  renderers [{rendererClass=&NiagaraSpriteRendererProperties}]
 }
 Ring {
  emitterUpdateScript [
   &EmitterState()
   &SpawnBurst_Instantaneous(Spawn Time=0.08, Use Spawn Probability=false, UseLoopCountLimit=false)
   &SpawnBurst_Instantaneous(Spawn Group=1, Spawn Time=0.08, Use Spawn Probability=false, UseLoopCountLimit=false) {module=SpawnBurst_Instantaneous001}
   &SpawnBurst_Instantaneous(Spawn Group=2, Spawn Time=0.08, Use Spawn Probability=false, UseLoopCountLimit=false) {module=SpawnBurst_Instantaneous002}
  ]
  particleSpawnScript [
   /SetVariables(Particles.VisibilityTag=&MaskIntBySpawnGroup(
     Masked Int=0
     Passthrough Int=&MaskIntBySpawnGroup(Masked Int=1, Passthrough Int=2, Spawn Group Mask=1)
     Spawn Group Mask=0
    )) {isSetParameters=true}
   &InitializeParticle(
    Color=&MaskLinearColorBySpawnGroup(Masked Linear Color=<rgba 3 3 3 1>, Passthrough Linear Color=<rgba 1 1 1 1>, Spawn Group Mask=2)
    Lifetime=0.6
    Mesh Scale=&Multiply_VectorByFloat(Vector=<1 1.7 1>, Float=0.2)
    Mesh Scale Mode=::NewEnumerator4
    Position Offset=&MaskVectorBySpawnGroup(
     Masked Vector=<0 0 0>
     Passthrough Vector=&MaskVectorBySpawnGroup(Masked Vector=<0 0 -25>, Passthrough Vector=<0 0 2.5>, Spawn Group Mask=1)
    )
    Sprite Size Mode=::NewEnumerator3
    Uniform Sprite Size=50
    UsePositionOffset=true
    Write Material Random=true
   )
   /SetVariables(Particles.GlowSpriteFacing:Vector3f=<0 0 1>) {isSetParameters=true}
   &InitialMeshOrientation(Mesh Orientation Mode=::NewEnumerator1, Rotation=<0.25 0 0>, Use Rotation Vector=true)
   &AddVelocity(Velocity=<0 0 -1>)
  ]
  particleUpdateScript [
   &ParticleState()
   &SolveForcesAndVelocity(Clamp Velocity=false, Limit Acceleration=false)
   &DynamicMaterialParameters(
    Index 0 Param 1=&FloatFromCurve(FloatCurve=DataInterface() {propertyValues {Curve {keys [{interpMode=RCIM_Cubic, leaveTangent=3.5412, tangentMode=RCTM_Break, time=0, value=0}, {tangentMode=RCTM_User, time=0.75, value=1}]}}})
    Module.Param0WriteEnabled=true
    Module.Param1WriteEnabled=true
    Module.Param2WriteEnabled=true
    Module.Param3WriteEnabled=true
   )
   &ScaleColor(
    Linear Color Curve=DataInterface() {propertyValues {
      AlphaCurve {keys [
        {arriveTangent=9.0725, interpMode=RCIM_Cubic, leaveTangent=9.0725, tangentMode=RCTM_User, time=0, value=0}
        {
         arriveTangent=-3.699
         arriveTangentWeight=0.0469
         interpMode=RCIM_Cubic
         leaveTangent=-3.699
         tangentMode=RCTM_User
         time=0.1408
         value=0.9847
        }
        {arriveTangent=-0.0405, arriveTangentWeight=0.2864, interpMode=RCIM_Cubic, leaveTangent=-0.0405, tangentMode=RCTM_User, time=1, value=0}
       ]}
      BlueCurve {keys [<at 0 0.1227>,<at 1 0.1451>]}
      GreenCurve {keys [<at 0 0.7537>,<at 1 0.4894>]}
      RedCurve {keys [<at 0 1>,<at 1 1>]}
     }}
    Scale Mode=::NewEnumerator2
   )
   &ScaleMeshSize(Scale Factor=&VectorFromFloat(Value=&FloatFromCurve(
      FloatCurve=DataInterface() {propertyValues {Curve {keys [
          {arriveTangent=5.6347, interpMode=RCIM_Cubic, leaveTangent=5.6347, tangentMode=RCTM_User, time=0, value=0.3788}
          {arriveTangent=1.1451, interpMode=RCIM_Cubic, leaveTangent=1.1451, time=0.2, value=0.85}
          {arriveTangent=0.2133, interpMode=RCIM_Cubic, leaveTangent=0.2133, tangentMode=RCTM_User, time=0.6186, value=1.0872}
          <at 0.9985 1.2296 RCIM_Cubic>
         ]}}}
      Scale Curve=0.95
     )))
   &UpdateMeshOrientation(Rotation Rate=0.2, Rotation Vector=<0 1 0>) {module=UpdateMeshOrientation001}
   /SetVariables(
    Particles.GlowColor=&ColorFromCurve(ColorCurve=DataInterface() {propertyValues {
       AlphaCurve {keys [
         {arriveTangent=9.0725, interpMode=RCIM_Cubic, leaveTangent=9.0725, tangentMode=RCTM_User, time=0, value=0}
         {
          arriveTangent=-1.6136
          arriveTangentWeight=0.0469
          interpMode=RCIM_Cubic
          leaveTangent=-1.6136
          tangentMode=RCTM_User
          time=0.1408
          value=0.6
         }
         {arriveTangent=-0.0405, arriveTangentWeight=0.2864, interpMode=RCIM_Cubic, leaveTangent=-0.0405, tangentMode=RCTM_User, time=1, value=0}
        ]}
       BlueCurve {keys [<at 0 0.1882 RCIM_Cubic>,<at 0.9969 0>]}
       GreenCurve {keys [<at 0 0.702 RCIM_Cubic>,<at 0.9969 0>]}
       RedCurve {keys [<at 0 1 RCIM_Cubic>,<at 0.9969 0>]}
      }})
    Particles.GlowSize=&Vector2DFromFloat(Value=&FloatFromCurve(FloatCurve=DataInterface() {propertyValues {Curve {keys [<at 0 0>,<at 0.4 1>]}}}, Scale Curve=65))
    Particles.SpeedlineColor:LinearColor=&ColorFromCurve(ColorCurve=DataInterface() {propertyValues {
       AlphaCurve {keys [
         {arriveTangent=9.0725, interpMode=RCIM_Cubic, leaveTangent=9.0725, tangentMode=RCTM_User, time=0, value=0}
         {
          arriveTangent=-3.699
          arriveTangentWeight=0.0469
          interpMode=RCIM_Cubic
          leaveTangent=-3.699
          tangentMode=RCTM_User
          time=0.1408
          value=0.9847
         }
         {arriveTangent=-0.0405, arriveTangentWeight=0.2864, interpMode=RCIM_Cubic, leaveTangent=-0.0405, tangentMode=RCTM_User, time=1, value=0}
        ]}
       BlueCurve {keys [<at 0 0 RCIM_Cubic>]}
       GreenCurve {keys [<at 0 0.749 RCIM_Cubic>]}
       RedCurve {keys [<at 0 1 RCIM_Cubic>]}
      }})
    Particles.SpeedlineSize:Vector2f=<35 90>
    Particles.WindColor:LinearColor=&ColorFromCurve(ColorCurve=DataInterface() {propertyValues {
       AlphaCurve {keys [
         {arriveTangent=9.0725, interpMode=RCIM_Cubic, leaveTangent=9.0725, tangentMode=RCTM_User, time=0, value=0}
         {arriveTangent=-1.6136, arriveTangentWeight=0.0469, interpMode=RCIM_Cubic, leaveTangent=-1.6136, tangentMode=RCTM_User, time=0.141, value=0.3}
         {arriveTangent=-0.0405, arriveTangentWeight=0.2864, interpMode=RCIM_Cubic, leaveTangent=-0.0405, tangentMode=RCTM_User, time=1, value=0}
        ]}
       BlueCurve {keys [<at 0 0.1882 RCIM_Cubic>,<at 0.9969 0>]}
       GreenCurve {keys [<at 0 0.702 RCIM_Cubic>,<at 0.9969 0>]}
       RedCurve {keys [<at 0 1 RCIM_Cubic>,<at 0.9969 0>]}
      }})
    Particles.WindDP:Vector4f=<0.5 0.6 0 0>
   ) {isSetParameters=true}
  ]
  rendererClasses [&NiagaraSpriteRendererProperties, &NiagaraMeshRendererProperties]
  renderers [
   {rendererClass=&NiagaraSpriteRendererProperties}
   {rendererClass=&NiagaraSpriteRendererProperties, rendererIndex=1}
   {rendererClass=&NiagaraSpriteRendererProperties, rendererIndex=2}
   {rendererClass=&NiagaraMeshRendererProperties, rendererIndex=3}
  ]
 }
 Sparkles {
  emitterUpdateScript [
   &EmitterState()
   &SpawnBurst_Instantaneous(Spawn Count=13, Spawn Time=0.08, Use Spawn Probability=false)
  ]
  particleSpawnScript [
   &InitializeParticle(
    Color Maximum=<rgba 15 11.676 3.325 1>
    Color Minimum=<rgba 15 9.4281 2.8125 1>
    Color Mode=::NewEnumerator2
    Lifetime=&UniformRangedFloat(Minimum=0.3, Maximum=1)
    Sprite Rotation Mode=::NewEnumerator2
    Sprite Size Mode=::NewEnumerator3
    Uniform Sprite Size=&UniformRangedFloat(Minimum=3, Maximum=8)
    UsePositionOffset=false
    Write Material Random=true
   ) {module=InitializeParticle001}
   &AddVelocity(Velocity=&UniformRangedVector(Minimum=<-200 -200 -60>, Maximum=<200 200 -325>))
   &ApplyInitialForces()
  ]
  particleUpdateScript [
   &ParticleState()
   &GravityForce(Gravity=<0 0 0>)
   /Niagara/Modules/Update/Forces/CurlNoiseForce(Noise Frequency=0.02, Noise Quality / Cost=::NewEnumerator0, Noise Strength=500, Pan Noise=false)
   &Drag(Drag=5, Use Linear Drag=true, Use Rotational Drag=false)
   &SolveForcesAndVelocity(Clamp Velocity=false, Limit Acceleration=false)
   &ScaleColor(
    Scale Alpha=&FloatFromCurve(FloatCurve=DataInterface() {propertyValues {Curve {keys [<at -0.0092 0.0144>,<at 0.0967 0.9767>,<at 1.0011 0.002>]}}})
    ScaleA=true
    ScaleRGB=true
   )
   &ScaleSpriteSize(Scale Factor=&Vector2DFromFloat(Value=&FloatFromCurve(FloatCurve=DataInterface() {propertyValues {Curve {keys [<at -0.0046 0>,<at 0.1746 0.9952>,<at 1.0011 0.002>]}}})))
  ]
  rendererClasses [&NiagaraSpriteRendererProperties]
  renderers [{rendererClass=&NiagaraSpriteRendererProperties}]
 }
}
summary {
 systemName=NS_OatmealSpread_Golem_Net
 emitters [
  {emitterName=Ring, rendererClasses [&NiagaraSpriteRendererProperties, &NiagaraMeshRendererProperties]}
  {emitterName=Sparkles, rendererClasses [&NiagaraSpriteRendererProperties]}
  {emitterName=CoreFlash, rendererClasses [&NiagaraMeshRendererProperties]}
  {emitterName=Ball, rendererClasses [&NiagaraSpriteRendererProperties]}
  {emitterName=Ribbon, rendererClasses [&NiagaraSpriteRendererProperties]}
 ]
 userVariables []
}
